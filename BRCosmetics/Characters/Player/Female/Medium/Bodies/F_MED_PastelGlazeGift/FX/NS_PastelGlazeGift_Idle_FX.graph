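# Idle cosmetic Niagara effect for the PastelGlazeGift female body: warm glows, light spores and pixie dust on the character's skeletal mesh.
# A mesh body glow and per-socket sprite glows track the skeleton, while spores and dust sample Glows particle positions at spawn.
# Two system timelines and time of day drive gameplay opacity, with grind/glide state also scaling body glow.
# Front end uses its own tuning and fade-in; quality level dims glows and dust.
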NiagaraSystem /BRCosmetics/Characters/Player/Female/Medium/Bodies/F_MED_PastelGlazeGift/FX/NS_PastelGlazeGift_Idle_FX
dependencies {
 usedDataInterfaces [
  &NiagaraDataInterfaceParticleRead
  &NiagaraDataInterfaceCurve
  &NiagaraDataInterfaceArrayFloat
  &NiagaraDataInterfaceVector2DCurve
  &NiagaraDataInterfaceArrayInt32
  &NiagaraDataInterfaceArrayFloat2
  &NiagaraDataInterfaceSkeletalMesh
 ]
 usedDynamicInputs [
  &SetBoolByFloatComparison
  &VectorLength
  &Multiply_Vector
  &Boolean_AndOperation
  &MakeFloatFromVector
  &TransformVector
  &FNTimeOfDayScalar
  &InvertBool
 ]
 usedModules [
  &self:SystemSpawnScript.SystemScriptSource.SystemScriptGraph.NiagaraNodeAssignment_0.SetVariables_#1
  &self:SystemSpawnScript.SystemScriptSource.SystemScriptGraph.NiagaraNodeAssignment_5.SetVariables_#2
  &self:SystemSpawnScript.SystemScriptSource.SystemScriptGraph.NiagaraNodeAssignment_8.SetVariables_#3
  &SystemState
  &self:SystemSpawnScript.SystemScriptSource.SystemScriptGraph.NiagaraNodeAssignment_6.SetVariables_#4
  &self:SystemSpawnScript.SystemScriptSource.SystemScriptGraph.NiagaraNodeAssignment_7.SetVariables_#5
  &Timeline
  &self:SystemSpawnScript.SystemScriptSource.SystemScriptGraph.NiagaraNodeAssignment_1.SetVariables_#6
  &self:SystemSpawnScript.SystemScriptSource.SystemScriptGraph.NiagaraNodeAssignment_2.SetVariables_#7
  &self:SystemSpawnScript.SystemScriptSource.SystemScriptGraph.NiagaraNodeAssignment_3.SetVariables_#8
  &self:SystemSpawnScript.SystemScriptSource.SystemScriptGraph.NiagaraNodeAssignment_10.SetVariables_#9
 ]
}
emitters {
 Em_Body_Glow {
  emitterUpdateScript [
   &EmitterState(
    Inactive Response=::NewEnumerator1
    Life Cycle Mode=::NewEnumerator1
    Loop Behavior=::NewEnumerator1
    Loop Duration Mode=::NewEnumerator1
    UseLoopDelay=false
   )
   &SpawnBurst_Instantaneous(Use Spawn Probability=false, UseLoopCountLimit=false)
  ]
  particleSpawnScript [&InitializeParticle(
    Color Mode=::NewEnumerator0
    Mesh Scale=&FN_FrontEndVector3Modifier(Gameplay=<3 1 1.25>, Front End=<2.2471 0.8629 0.7922>)
    Mesh Scale Mode=::NewEnumerator4
    UsePositionOffset=false
    Write Material Random=true
   )]
  particleUpdateScript [
   &ParticleState(Kill Particles When Lifetime Has Elapsed=false, Let Infinitely Lived Particles Die When Emitter Deactivates=true)
   &SkeletalMeshLocation(
    Bone / Socket Index=&SelectIntFromArray(Direct Array Index=0, Int Selection Array=DataInterface() {propertyValues {IntData [0]}})
    Bone Sampling Mode=::NewEnumerator4
    MaskSpawnGroup=false
    Skeletal Mesh=$System.SkeletalMesh
   )
   &CameraOffset()
   &ScaleColor(
    Scale Alpha=&FN_FrontEndFloatModifier(
     Gameplay=&Multiply_Float(
      A=&Multiply_Float(
       A=&Add_Float(
        A=&Multiply_Float(
         A=&FloatFromCurve(FloatCurve=DataInterface() {propertyValues {Curve {keys [
              {arriveTangent=1.2826, interpMode=RCIM_Cubic, leaveTangent=1.2826, tangentMode=RCTM_User, time=0, value=0}
              {tangentMode=RCTM_User, time=1, value=1}
             ]}}})
         B=&Lerp_Float(A=0, B=1, Alpha=$System.Timeline.NormalizedTime)
        )
        B=&Multiply_Float(
         A=&FloatFromCurve(FloatCurve=DataInterface() {propertyValues {Curve {keys [
              {arriveTangent=1.2826, interpMode=RCIM_Cubic, leaveTangent=1.2826, tangentMode=RCTM_User, time=0, value=0}
              {tangentMode=RCTM_User, time=1, value=0.5}
             ]}}})
         B=&Lerp_Float(A=0, B=1, Alpha=$System.Timeline001.NormalizedTime)
        )
       )
       B=&FloatFromCurve(
        FloatCurve=DataInterface() {propertyValues {Curve {keys [
            {interpMode=RCIM_Cubic, tangentMode=RCTM_User, time=0, value=0.7}
            {interpMode=RCIM_Cubic, tangentMode=RCTM_User, time=0.5, value=1.3}
            {tangentMode=RCTM_User, time=1, value=0.7}
           ]}}}
        CurveIndex=Linked() {linkedVariable {name=NPC.FortniteNPC.TimeOfDay_0-1, type=NiagaraFloat}}
       )
      )
      B=&MakeCustomFloatFromBool(
       Bool=$System.IsGrinding
       True Float=&MakeCustomFloatFromBool(Bool=$System.IsGliding, True Float=1, False Float=3)
       False Float=1
      )
     )
     Front End=&Multiply_Float(A=$System.FadeIn_Frontend, B=5)
    )
    ScaleA=true
    ScaleRGB=true
   ) {module=ScaleColor002}
   /SetVariables(Particles.MeshOrientationFE:Quat4f=&LerpQuaternion(
     Quaternion A:Quat4f=$Particles.MeshOrientation
     Quaternion B:Quat4f=<0 1 0 1>
     Lerp Factor:NiagaraFloat=1
     Enable Spherical Linear Interpolation:NiagaraBool=false
    )) {isSetParameters=true}
  ]
  rendererClasses [&NiagaraMeshRendererProperties]
  renderers [
   {rendererClass=&NiagaraMeshRendererProperties, rendererIndex=0}
   {rendererClass=&NiagaraMeshRendererProperties, rendererIndex=1}
  ]
 }
 Glows {
  emitterUpdateScript [
   &EmitterState(
    Inactive Response=::NewEnumerator1
    Life Cycle Mode=::NewEnumerator1
    Loop Behavior=::NewEnumerator1
    Loop Duration Mode=::NewEnumerator1
    UseLoopDelay=false
   )
   &SpawnBurst_Instantaneous(Spawn Count=$System.SpawnGroups, Use Spawn Probability=true, UseLoopCountLimit=false)
  ]
  particleSpawnScript [
   &FN_SetSpawnGroup(Number of Spawn Groups:NiagaraInt32=$System.SpawnGroups)
   &InitializeParticle(
    Color=<rgba 5 3.4933 0.8824 0.05>
    Sprite Size=&FN_FrontEndVector2Modifier(
     Front End:Vector2f=&SelectVector2DFromArray(
      Direct Array Index=$Particles.SpawnGroup
      Vector 2D Selection Array=DataInterface() {propertyValues {FloatData [<11 11>, <10 10>, <10 10>, <10 10>, <9 9>, <10 10>, <10 10>, <6 6>, <10 10>]}}
     )
     Gameplay:Vector2f=&SelectVector2DFromArray(
      Direct Array Index=$Particles.SpawnGroup
      Vector 2D Selection Array=DataInterface() {propertyValues {FloatData [<10 15>, <10 15>, <10 15>, <10 15>, <9 11>, <8 15>, <9 14>, <10 15>, <12 16>]}}
     )
    )
    Sprite Size Mode=::NewEnumerator4
    UsePositionOffset=false
    Write Material Random=true
   )
  ]
  particleUpdateScript [
   &SolveForcesAndVelocity(Clamp Velocity=false, Limit Acceleration=false)
   &ParticleState(Kill Particles When Lifetime Has Elapsed=false, Let Infinitely Lived Particles Die When Emitter Deactivates=true)
   &SkeletalMeshLocation(
    Bone / Socket Index=&SelectIntFromArray(Direct Array Index=$Particles.SpawnGroup, Int Selection Array=$System.Sockets)
    MaskSpawnGroup=false
    Mesh Sampling Type=::NewEnumerator5
    Skeletal Mesh=$System.SkeletalMesh
    Socket Sampling Mode=::NewEnumerator5
   ) {module=SkeletalMeshLocation003}
   &CameraOffset(Camera Offset Amount=&FN_FrontEndFloatModifier(Gameplay=2, Front End=3))
   &ScaleColor(
    Scale Alpha=&Multiply_Float(
     A=&FN_FrontEndFloatModifier(
      Gameplay=&Multiply_Float(
       A=&Add_Float(
        A=&Multiply_Float(
         A=&FloatFromCurve(FloatCurve=DataInterface() {propertyValues {Curve {keys [
              {arriveTangent=0.7839, interpMode=RCIM_Cubic, leaveTangent=0.7839, tangentMode=RCTM_User, time=0, value=0}
              {tangentMode=RCTM_User, time=1, value=0.6}
             ]}}})
         B=&Lerp_Float(A=0, B=1, Alpha=$System.Timeline.NormalizedTime)
        )
        B=&Multiply_Float(
         A=&FloatFromCurve(FloatCurve=DataInterface() {propertyValues {Curve {keys [
              {arriveTangent=1.2826, interpMode=RCIM_Cubic, leaveTangent=1.2826, tangentMode=RCTM_User, time=0, value=0}
              {tangentMode=RCTM_User, time=1, value=0.6}
             ]}}})
         B=&Lerp_Float(A=0, B=1, Alpha=$System.Timeline001.NormalizedTime)
        )
       )
       B=&FloatFromCurve(
        FloatCurve=DataInterface() {propertyValues {Curve {keys [
            {interpMode=RCIM_Cubic, tangentMode=RCTM_User, time=0, value=0.7}
            {interpMode=RCIM_Cubic, tangentMode=RCTM_User, time=0.5, value=1.3}
            {tangentMode=RCTM_User, time=1, value=0.7}
           ]}}}
        CurveIndex=Linked() {linkedVariable {name=NPC.FortniteNPC.TimeOfDay_0-1, type=NiagaraFloat}}
        Scale Curve=2
       )
      )
      Front End=&Multiply_Float(A=$System.FadeIn_Frontend, B=3)
     )
     B=&FN_FrontEndFloatModifier(
      Gameplay=&SelectFloatFromArray(Direct Array Index=$Engine.QualityLevel, Float Selection Array=DataInterface() {propertyValues {FloatData [0.25,0.5,1,1,1]}})
      Front End=&SelectFloatFromArray(Direct Array Index=$Engine.QualityLevel, Float Selection Array=DataInterface() {propertyValues {FloatData [0.1,0.5,1,1,1]}})
     )
    )
    ScaleA=true
    ScaleRGB=false
   )
  ]
  rendererClasses [&NiagaraSpriteRendererProperties]
  renderers [{rendererClass=&NiagaraSpriteRendererProperties}]
 }
 LightSpores {
  emitterUpdateScript [
   &EmitterState(
    Inactive Response=::NewEnumerator1
    Life Cycle Mode=::NewEnumerator1
    Loop Behavior=::NewEnumerator1
    Loop Duration Mode=::NewEnumerator1
    UseLoopDelay=false
   )
   &SpawnRate(SpawnRate=33, Use Spawn Probability=true) {module=SpawnRate001}
  ]
  particleSpawnScript [
   &FN_SetSpawnGroup(Number of Spawn Groups:NiagaraInt32=$System.SpawnGroups)
   &self:SamplePositionFromOtherParticle(PAR=DataInterface() {propertyValues {EmitterBinding {bindingMode=Other, emitterName=Glows}}}) {module=SamplePositionFromOtherParticle}
   &InitializeParticle(
    Color=<rgba 1 0.3226 0 1>
    Lifetime Max=&FN_FrontEndFloatModifier(Gameplay=0.6, Front End=1)
    Lifetime Min:NiagaraFloat=&FN_FrontEndFloatModifier(Gameplay=0.4, Front End=0.6)
    Lifetime Mode=::NewEnumerator1
    Sprite Size Mode=::NewEnumerator3
    UsePositionOffset=false
    Write Material Random=true
   )
   &SkeletalMeshLocation(MaskSpawnGroup=false, Mesh Sampling Type=::NewEnumerator5, OverrideRandomSeed=false, Skeletal Mesh=$System.SkeletalMesh) {module=SkeletalMeshLocation003}
   &ShapeLocation(
    MaskSpawnGroup=false
    Override Seed=false
    Shape Origin=$StackContext.SkeletalMeshLocation003.SampledPosition
    Sphere Radius=5
    UseOwnerScale=true
   )
   &AddVelocity(
    Module.Use Falloff Distance=false
    Override Seed=false
    Velocity Mode=::NewEnumerator1
    Velocity Speed=&FN_FrontEndFloatModifier(
     Gameplay=&RandomRangeFloat(Minimum=22, Maximum=30)
     Front End=&RandomRangeFloat(Minimum=13, Maximum=17)
    )
   )
  ]
  particleUpdateScript [
   &ParticleState()
   &InheritSourceMovement(
    Applied Position Scale=&VectorFromFloat(Value=&FloatFromCurve(FloatCurve=DataInterface() {propertyValues {Curve {keys [{interpMode=RCIM_Cubic, tangentMode=RCTM_User, time=0, value=1}, {interpMode=RCIM_Cubic, tangentMode=RCTM_User, time=1, value=0}]}}}))
    ApplyForce=false
    ApplyPosition=true
    ApplyPositionFalloff=false
   )
   &Drag(
    Drag=&FloatFromCurve(
     FloatCurve=DataInterface() {propertyValues {Curve {keys [{interpMode=RCIM_Cubic, tangentMode=RCTM_User, time=0, value=0}, {interpMode=RCIM_Cubic, tangentMode=RCTM_User, time=1, value=1}]}}}
     Scale Curve=&RandomRangeFloat(Minimum=0, Maximum=5)
    )
    Use Linear Drag=true
    Use Rotational Drag=true
   ) {module=Drag001}
   &CameraOffset(Camera Offset Amount=&RandomRangeFloat(Minimum=0, Maximum=10))
   &SolveForcesAndVelocity(Clamp Velocity=false, Limit Acceleration=false)
   &ScaleSpriteSize(
    Uniform Curve Scale=&Waveform(
     "Waveform [1]"=::NewEnumerator0
     "[1] Frequency"=&RandomRangeFloat(Minimum=0, Maximum=1.8831)
     "[1] Amplitude Min/Max"=<0.7 1>
    )
    Uniform Curve Sprite Scale=DataInterface() {propertyValues {Curve {keys [
        {interpMode=RCIM_Cubic, tangentMode=RCTM_User, time=0, value=1}
        <at 0.9 1>
        {arriveTangent=-4.3352, interpMode=RCIM_Cubic, leaveTangent=-4.3352, tangentMode=RCTM_User, time=1, value=0}
       ]}}}
   ) {module=ScaleSpriteSize001}
   &ScaleColor(
    Scale Alpha=&FN_FrontEndFloatModifier(
     Gameplay=&Multiply_Float(
      A=&Add_Float(
       A=&Multiply_Float(
        A=&FloatFromCurve(FloatCurve=DataInterface() {propertyValues {Curve {keys [{interpMode=RCIM_Cubic, tangentMode=RCTM_User, time=0, value=1}, {interpMode=RCIM_Cubic, tangentMode=RCTM_User, time=1, value=0}]}}})
        B=&Lerp_Float(A=0, B=0.25, Alpha=$System.Timeline.NormalizedTime)
       )
       B=&Multiply_Float(
        A=&FloatFromCurve(FloatCurve=DataInterface() {propertyValues {Curve {keys [{interpMode=RCIM_Cubic, tangentMode=RCTM_User, time=0, value=1}, {interpMode=RCIM_Cubic, tangentMode=RCTM_User, time=1, value=0}]}}})
        B=&Lerp_Float(A=0, B=1.5, Alpha=$System.Timeline001.NormalizedTime)
       )
      )
      B=&FloatFromCurve(
       FloatCurve=DataInterface() {propertyValues {Curve {keys [
           {interpMode=RCIM_Cubic, tangentMode=RCTM_User, time=0, value=0.7}
           {interpMode=RCIM_Cubic, tangentMode=RCTM_User, time=0.5, value=1.3}
           {tangentMode=RCTM_User, time=1, value=0.7}
          ]}}}
       CurveIndex=Linked() {linkedVariable {name=NPC.FortniteNPC.TimeOfDay_0-1, type=NiagaraFloat}}
      )
     )
     Front End=&Multiply_Float(
      A=&Multiply_Float(
       A=&FloatFromCurve(FloatCurve=DataInterface() {propertyValues {Curve {keys [{interpMode=RCIM_Cubic, tangentMode=RCTM_User, time=0, value=1}, {interpMode=RCIM_Cubic, tangentMode=RCTM_User, time=1, value=0}]}}})
       B=$System.FadeIn_Frontend
      )
      B=0.5
     )
    )
    ScaleA=true
    ScaleRGB=false
   ) {module=ScaleColor001}
  ]
  rendererClasses [&NiagaraSpriteRendererProperties]
  renderers [
   {rendererClass=&NiagaraSpriteRendererProperties}
   {rendererClass=&NiagaraSpriteRendererProperties, rendererIndex=1}
  ]
 }
 PixieDust {
  emitterUpdateScript [
   &EmitterState(
    Inactive Response=::NewEnumerator1
    Life Cycle Mode=::NewEnumerator1
    Loop Behavior=::NewEnumerator1
    Loop Duration Mode=::NewEnumerator1
    UseLoopDelay=false
   )
   &SpawnRate(SpawnRate=24, Use Spawn Probability=true)
  ]
  particleSpawnScript [
   &FN_SetSpawnGroup(Number of Spawn Groups:NiagaraInt32=$System.SpawnGroups)
   &self:SamplePositionFromOtherParticle(PAR=DataInterface() {propertyValues {EmitterBinding {bindingMode=Other, emitterName=Glows}}}) {module=SamplePositionFromOtherParticle001}
   &InitializeParticle(
    Color Maximum=<rgba 0.808 0.4909 0.1856 1>
    Color Minimum=<rgba 0.9445 0.5331 0.1371 1>
    Color Mode=::NewEnumerator2
    Lifetime=&FN_FrontEndFloatModifier(
     Gameplay=&UniformRangedFloat(Minimum=0.5, Maximum=0.75)
     Front End=&UniformRangedFloat(Minimum=1, Maximum=1.25)
    )
    Sprite Rotation Mode=::NewEnumerator2
    Sprite Size Mode=::NewEnumerator1
    Uniform Sprite Size Max=25
    Uniform Sprite Size Min=10
    UsePositionOffset=false
    Write Material Random=true
   )
   &DynamicMaterialParameters(
    Index 0 Param 1=&RandomRangeFloat(Minimum=-0.05, Maximum=0.05)
    Index 0 Param 2=&RandomRangeFloat(Minimum=0, Maximum=32)
    Module.Index0Param0Write=true
    Module.Index0Param1Write=true
    Module.Index0Param2Write=false
    Module.Index0Param3Write=false
   )
   &AddVelocity(
    Velocity=&FN_FrontEndVector3Modifier(Gameplay=<0 0 0>, Front End=<0 0 1>)
    Velocity Speed Scale=&RandomRangeFloat(Minimum=10, Maximum=13)
   )
  ]
  particleUpdateScript [
   &ParticleState()
   &SkeletalMeshLocation(
    MaskSpawnGroup=false
    Mesh Sampling Type=::NewEnumerator5
    OverrideRandomSeed=false
    Position Sampling=::NewEnumerator1
    Skeletal Mesh=$System.SkeletalMesh
    Source Movement Scale:Vector3f=&VectorFromFloat(Value=0)
   ) {module=SkeletalMeshLocation001}
   &InheritSourceMovement(
    Applied Position Scale=&VectorFromFloat(Value=&FloatFromCurve(FloatCurve=DataInterface() {propertyValues {Curve {keys [<at 0 0.85 RCIM_Cubic>,<at 0.5 1>]}}}))
    ApplyForce=false
    ApplyPosition=true
    ApplyPositionFalloff=false
   ) {module=InheritSourceMovement001}
   &SolveForcesAndVelocity(Clamp Velocity=false, Limit Acceleration=false)
   &ScaleColor(
    Scale Alpha=&Multiply_Float(
     A=&FN_FrontEndFloatModifier(
      Gameplay=&Multiply_Float(
       A=&FloatFromCurve(
        FloatCurve=DataInterface() {propertyValues {Curve {keys [
            {arriveTangent=9.0725, interpMode=RCIM_Cubic, leaveTangent=9.0725, tangentMode=RCTM_User, time=0, value=0}
            {arriveTangent=-0.5016, interpMode=RCIM_Cubic, leaveTangent=-0.5016, tangentMode=RCTM_User, time=0.1408, value=0.9847}
            {arriveTangent=-0.0405, interpMode=RCIM_Cubic, leaveTangent=-0.0405, tangentMode=RCTM_User, time=1, value=0}
           ]}}}
        Scale Curve=0.75
       )
       B=&Multiply_Float(
        A=&Add_Float(
         A=&Multiply_Float(
          A=&FloatFromCurve(FloatCurve=DataInterface() {propertyValues {Curve {keys [{interpMode=RCIM_Cubic, tangentMode=RCTM_User, time=0, value=1}, {interpMode=RCIM_Cubic, tangentMode=RCTM_User, time=1, value=0}]}}})
          B=&Lerp_Float(A=0, B=1, Alpha=$System.Timeline.NormalizedTime)
         )
         B=&Multiply_Float(
          A=&FloatFromCurve(FloatCurve=DataInterface() {propertyValues {Curve {keys [{interpMode=RCIM_Cubic, tangentMode=RCTM_User, time=0, value=1}, {interpMode=RCIM_Cubic, tangentMode=RCTM_User, time=1, value=0}]}}})
          B=&Lerp_Float(A=0, B=1, Alpha=$System.Timeline001.NormalizedTime)
         )
        )
        B=&FloatFromCurve(
         FloatCurve=DataInterface() {propertyValues {Curve {keys [
             {interpMode=RCIM_Cubic, tangentMode=RCTM_User, time=0, value=0.7}
             {interpMode=RCIM_Cubic, tangentMode=RCTM_User, time=0.5, value=1.3}
             {tangentMode=RCTM_User, time=1, value=0.7}
            ]}}}
         CurveIndex=Linked() {linkedVariable {name=NPC.FortniteNPC.TimeOfDay_0-1, type=NiagaraFloat}}
        )
       )
      )
      Front End=&Multiply_Float(
       A=&Multiply_Float(
        A=&FloatFromCurve(
         FloatCurve=DataInterface() {propertyValues {Curve {keys [
             {arriveTangent=9.0725, interpMode=RCIM_Cubic, leaveTangent=9.0725, tangentMode=RCTM_User, time=0, value=0}
             {arriveTangent=-0.5016, interpMode=RCIM_Cubic, leaveTangent=-0.5016, tangentMode=RCTM_User, time=0.1408, value=0.9847}
             {arriveTangent=-0.0405, interpMode=RCIM_Cubic, leaveTangent=-0.0405, tangentMode=RCTM_User, time=1, value=0}
            ]}}}
         Scale Curve=0.75
        )
        B=$System.FadeIn_Frontend
       )
       B=0.5
      )
     )
     B=&FN_FrontEndFloatModifier(
      Gameplay=&SelectFloatFromArray(Direct Array Index=$Engine.QualityLevel, Float Selection Array=DataInterface() {propertyValues {FloatData [0.25,1,1,1,1]}})
      Front End=&SelectFloatFromArray(Direct Array Index=$Engine.QualityLevel, Float Selection Array=DataInterface() {propertyValues {FloatData [0.5,1,1,1,1]}})
     )
    )
    ScaleA=true
    ScaleRGB=true
   )
   &ScaleSpriteSize(Uniform Curve Sprite Scale=DataInterface() {propertyValues {Curve {keys [
        {arriveTangent=2.2332, interpMode=RCIM_Cubic, leaveTangent=2.2332, tangentMode=RCTM_User, time=0, value=0.6}
        {tangentMode=RCTM_User, time=1, value=1.25}
       ]}}})
  ]
  rendererClasses [&NiagaraSpriteRendererProperties]
  renderers [{rendererClass=&NiagaraSpriteRendererProperties}]
 }
}
summary {
 systemName=NS_PastelGlazeGift_Idle_FX
 emitters [
  {emitterName=LightSpores, rendererClasses [&NiagaraSpriteRendererProperties]}
  {emitterName=PixieDust, rendererClasses [&NiagaraSpriteRendererProperties]}
  {emitterName=Em_Body_Glow, rendererClasses [&NiagaraMeshRendererProperties]}
  {emitterName=Glows, rendererClasses [&NiagaraSpriteRendererProperties]}
 ]
 userVariables [
  User.bIsFrontEnd {type=NiagaraBool}
  User.bIsFrontEndPreview {type=NiagaraBool}
  User.bisInVehicle {type=NiagaraBool}
  User.FrontEndReactivity {type=NiagaraBool}
  User.IsFrontend {type=NiagaraBool}
  User.IsSkydiving {type=NiagaraBool}
  User.VelocityThresholdFreefalling {type=NiagaraFloat, defaultValue=750, description=""}
  User.VelocityThresholdGlidingForward {type=NiagaraFloat, defaultValue=1000, description=""}
  User.VelocityThresholdGrinding {type=NiagaraFloat, defaultValue=850, description=""}
  User.VelocityThresholdRunning {type=NiagaraFloat, defaultValue=400, description=""}
  User.VelocityThresholdSprinting {type=NiagaraFloat, defaultValue=600, description=""}
 ]
}
