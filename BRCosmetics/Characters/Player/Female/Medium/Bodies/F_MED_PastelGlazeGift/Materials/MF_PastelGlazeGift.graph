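# Material function for the PastelGlazeGift female outfit that layers animated, reactive glitter onto incoming material attributes, behind a static switch.
# Three offset samples of a tiled mask texture blink on sine and cosine phases, broken up by panning Perlin noise in pre-skinned local space.
# Channel mask, Fresnel and reactivity shape the glitter, which alters color, emissive, metallic, specular, roughness and flattens normals.
# The result then passes through MF_NewDiamond_Cheap, borrowed from the BerryTartBrunt outfit.
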MaterialFunction /BRCosmetics/Characters/Player/Female/Medium/Bodies/F_MED_PastelGlazeGift/Materials/MF_PastelGlazeGift
graph {nodes [
  n1 = GetMaterialAttributes(None=NamedRerouteDeclaration(None=FunctionInput() {inputType=FunctionInput_MaterialAttributes}) {name=MA_IN, nodeColor=<rgba 1 0.9141 0 1>}) {attributeGetTypes [
    #1
    #2
    #3
    #4
    #5
    #6
   ]}
  n2 = NamedRerouteUsage()
  n3 = NamedRerouteUsage()
  n4 = Constant() {r=0.3}
  n5 = NamedRerouteUsage()
  n6 = SetMaterialAttributes!(
   MaterialAttributes=@n1.MaterialAttributes
   BaseColor=LinearInterpolate(A=@n1.BaseColor, B=@n2, Alpha=@n3)
   EmissiveColor=LinearInterpolate(A=@n1.EmissiveColor, B=Multiply(A=@n2, B=@n4), Alpha=@n3)
   Metallic=LinearInterpolate(A=@n1.Metallic, B=Constant() {r=1}, Alpha=@n5)
   Normal=NamedRerouteUsage()
   Specular=LinearInterpolate(A=@n1.Specular, B=Constant() {r=0.5}, Alpha=@n5)
   Roughness=LinearInterpolate(A=@n1.Roughness, B=Constant(), Alpha=@n5)
  ) {attributeSetTypes [
    #1
    #2
    #3
    #6
    #4
    #5
   ]}
  n7 = NamedRerouteUsage()
  FunctionOutput(None=StaticSwitchParameter(
    True=MaterialFunctionCall!(Material=@n6, TechArtMask=@n7) {materialFunction=/BRCosmetics/Characters/Player/Female/Medium/Bodies/F_MED_BerryTartBrunt/Materials/MF_NewDiamond_Cheap}.Material
    False=NamedRerouteUsage()
   ) {defaultValue=true, group=PastelGlazeGift Glitter, parameterName=UsePastelGlazeGift, sortPriority=0})
  NamedRerouteDeclaration(None=Multiply(
    A=ScalarParameter() {group=PastelGlazeGift Glitter, parameterName=PGG_Reactivity, sliderMin=1}
    B=ScalarParameter() {defaultValue=1, group=PastelGlazeGift Glitter, parameterName=PGG_ReactivityToggle, sliderMin=1}
   )) {name=Reactivity, nodeColor=<rgba 0 0.9766 1 1>}
  n8 = Add(A=NamedRerouteUsage(), B=NamedRerouteUsage())
  n9 = NamedRerouteUsage()
  n10 = Multiply(A=Saturate(None=Add(A=@n8, B=@n9)), B=NamedRerouteUsage())
  n11 = Fresnel() {baseReflectFraction=0, exponent=0.75}
  n12 = LinearInterpolate(A=OneMinus(None=@n11), B=Add(A=@n11, B=Constant() {r=0.4}), Alpha=NamedRerouteUsage())
  n13 = NamedRerouteUsage()
  n14 = Power(
   Base=Saturate(None=Subtract(A=@n12, B=@n13) {constB=0})
   Exp=ScalarParameter() {defaultValue=5, group=PastelGlazeGift Glitter, parameterName=Glitter_FresnelPower_Idle}
  )
  n15 = ScalarParameter() {defaultValue=0.6, group=PastelGlazeGift Glitter, parameterName=Glitter_Intensity_Idle}
  n16 = ScalarParameter() {defaultValue=6, group=PastelGlazeGift Glitter, parameterName=Glitter_Intensity_Reactive}
  n17 = NamedRerouteUsage()
  NamedRerouteDeclaration(None=Multiply(A=Multiply(A=@n10, B=@n14), B=LinearInterpolate(A=@n15, B=@n16, Alpha=@n17))) {name=GlitterMask, nodeColor=<rgba 0.5104 0.2648 0 1>}
  n18 = Add(A=NamedRerouteUsage(), B=NamedRerouteUsage())
  n19 = NamedRerouteUsage()
  NamedRerouteDeclaration(None=Saturate(None=Add(A=@n18, B=@n19))) {name=GlitterAlpha, nodeColor=<rgba 1 0 0.375 1>}
  n20 = ChannelMaskParameter(None=FunctionInput() {inputName=TechArtMask, sortPriority=1}) {group=PastelGlazeGift Glitter, maskChannel=Green, parameterName=Glitter_Channel, sortPriority=1, defaultValue=<rgba 0 1 0 0>}
  NamedRerouteDeclaration(None=Power(Base=@n20.R) {constExponent=0.75}) {name=GlitterChannel, nodeColor=<rgba 0 0.4609 1 1>}
  n21 = NamedRerouteUsage()
  n22 = NamedRerouteUsage()
  n23 = TextureObject() {
   samplerType=SAMPLERTYPE_Masks
   texture=/BRCosmetics/Characters/Player/Female/Medium/Bodies/F_MED_PastelGlazeGift/Textures/T_PastelGlazeGift_Body_TA
  }
  n24 = TextureSample(UVs=Add(A=NamedRerouteUsage(), B=ConstantBiasScale(None=@n22) {bias=-0.5, scale=0.1}), Tex=@n23) {
   samplerSource=SSM_Wrap_WorldGroupSettings
   samplerType=SAMPLERTYPE_Masks
   texture=&T_PastelGlazeGift_Body_TA
  }
  n25 = ConstantBiasScale(None=Cosine(None=Add(A=@n21, B=@n24.R) {constB=0}))
  n26 = NamedRerouteDeclaration(None=@n24.G) {name=Alpha1, nodeColor=<rgba 0.9453 1 0 1>}
  n27 = NamedRerouteUsage()
  NamedRerouteDeclaration(None=Subtract(A=Multiply(A=@n25, B=@n26), B=Multiply(A=@n27) {constB=0.5})) {name=Random1, nodeColor=<rgba 1 0 0.4453 1>}
  n28 = Time()
  n29 = ScalarParameter() {defaultValue=0.5, group=PastelGlazeGift Glitter, parameterName=Glitter_BlinkSpeed_Idle}
  NamedRerouteDeclaration(None=Frac(None=Multiply(A=@n28, B=@n29))) {name=Time, nodeColor=<rgba 0 0.625 1 1>}
  n30 = Divide(A=NamedRerouteUsage(), B=Constant() {r=50}) {constB=64}
  n31 = TextureObject() {texture=/Game/Effects/Fort_Effects/Textures/Tile/T_PerlinRGB_01_M}
  n32 = TextureSample(UVs=Panner(Coordinate=ComponentMask(None=@n30) {b=true, r=true}) {speedX=0.035, speedY=0.05}, Tex=@n31) {samplerSource=SSM_Wrap_WorldGroupSettings}
  NamedRerouteDeclaration(None=@n32.R) {name=Noise1, nodeColor=<rgba 1 0 0.8437 1>}
  n33 = TextureSample(UVs=Panner(Coordinate=ComponentMask(None=@n30) {b=true, g=true}) {speedX=-0.024, speedY=-0.034}, Tex=@n31) {samplerSource=SSM_Wrap_WorldGroupSettings}
  NamedRerouteDeclaration(None=@n33.G) {name=Noise2, nodeColor=<rgba 0 1 0.8359 1>}
  NamedRerouteDeclaration() {name=Debug, nodeColor=<rgba 0 0.9297 1 1>}
  n34 = Add(A=NamedRerouteUsage()) {constB=0.75, desc=Offset}
  n35 = NamedRerouteUsage()
  n36 = Constant2Vector() {g=0.856, r=0.435}
  n37 = NamedRerouteUsage()
  n38 = TextureSample(UVs=Add(A=Add(A=@n35, B=@n36) {constB=0.25}, B=ConstantBiasScale(None=@n37) {bias=-0.5, scale=0.25}), Tex=@n23) {
   samplerSource=SSM_Wrap_WorldGroupSettings
   samplerType=SAMPLERTYPE_Masks
   texture=&T_PastelGlazeGift_Body_TA
  }
  n39 = ConstantBiasScale(None=Sine(None=Add(A=@n34, B=@n38.R)))
  n40 = NamedRerouteDeclaration(None=@n38.G) {name=Alpha2, nodeColor=<rgba 0.7344 0 1 1>}
  n41 = NamedRerouteUsage()
  NamedRerouteDeclaration(None=Subtract(A=Multiply(A=@n39, B=@n40) {constB=0.5}, B=Multiply(A=@n41) {constB=0.5})) {name=Random2, nodeColor=<rgba 1 0 0.8906 1>}
  NamedRerouteDeclaration(None=Multiply(A=Multiply(A=@n32.G, B=@n33.B)) {constB=0.5}) {name=PanningNoise, nodeColor=<rgba 0 0.7891 1 1>}
  NamedRerouteDeclaration(None=Multiply(A=@n32.B, B=@n33.R)) {name=Noise3, nodeColor=<rgba 1 0.8906 0 1>}
  n42 = Add(A=NamedRerouteUsage()) {constB=1.5, desc=Offset}
  n43 = TextureSample(UVs=Add(A=NamedRerouteUsage(), B=Constant2Vector() {g=0.123, r=0.324}) {constB=0.5}, Tex=@n23) {
   samplerSource=SSM_Wrap_WorldGroupSettings
   samplerType=SAMPLERTYPE_Masks
   texture=&T_PastelGlazeGift_Body_TA
  }
  n44 = ConstantBiasScale(None=Sine(None=Add(A=@n42, B=@n43.R)))
  n45 = NamedRerouteDeclaration(None=@n43.G) {name=Alpha3, nodeColor=<rgba 0.0547 1 0 1>}
  n46 = NamedRerouteUsage()
  NamedRerouteDeclaration(None=Subtract(A=Multiply(A=@n44, B=@n45) {constB=0.5}, B=Multiply(A=@n46) {constB=0.5})) {name=Random3, nodeColor=<rgba 1 0 0.9375 1>}
  NamedRerouteDeclaration(None=Multiply(A=TextureCoordinate(), B=ScalarParameter() {defaultValue=7, group=PastelGlazeGift Glitter, parameterName=Glitter_Tiling})) {name=UVs, nodeColor=<rgba 0 0.8125 1 1>}
  NamedRerouteDeclaration(None=VertexInterpolator(VS=LocalPosition() {localOrigin=InstancePreSkinning}.XYZ).PS) {name=PS-LocalPos, nodeColor=<rgba 1 0.6328 0 1>}
  NamedRerouteDeclaration(None=VertexInterpolator(VS=PreSkinnedNormal()).PS) {name=PS-LocalNormal, nodeColor=<rgba 0.3125 0 1 1>}
  n47 = TextureSample() {
   samplerSource=SSM_Wrap_WorldGroupSettings
   samplerType=SAMPLERTYPE_Masks
   texture=&T_PastelGlazeGift_Body_TA
  }
  NamedRerouteDeclaration(None=@n47.B) {name=DiamondChannel, nodeColor=<rgba 1 0 0.5156 1>}
  NamedRerouteDeclaration(None=VectorParameter() {group=PastelGlazeGift Glitter, parameterName=Glitter_Color, defaultValue=<rgba 1 0.764 0 1>}.RGB) {name=GlitterColor, nodeColor=<rgba 1 0 0.1875 1>}
  NamedRerouteDeclaration(None=@n1.Normal) {name=Normal_IN, nodeColor=<rgba 0.1953 1 0 1>}
  n48 = Multiply(A=@n10) {constB=2}
  NamedRerouteDeclaration(None=LinearInterpolate(A=NamedRerouteUsage(), B=Constant3Vector() {constant=<rgba 0 0 1 0>}, Alpha=Saturate(None=@n48))) {name=Normal_OUT, nodeColor=<rgba 1 0 0.5156 1>}
  NamedRerouteUsage()
 ]}
