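# Physics asset for the hand cuffs of the PastelGlazeGift female medium body cosmetic, defining simulation bodies for the fur bones around both wrists.
# Each forearm twist bone has a long capsule anchoring four small dyn_fur capsules, named by compass direction.
# Each fur body constrains directly to its forearm with identical tight swing and twist limits, so pieces move slightly without chaining together.
# Right-side fur bones lack a side suffix, except dyn_fur_e_r.
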PhysicsAsset /BRCosmetics/Characters/Player/Female/Medium/Bodies/F_MED_PastelGlazeGift/Meshes/F_MED_PastelGlazeGift_Hand_Cuffs_Physics
bodies {
 dyn_fur_e_l [{shapeType=Capsule, length=2, radius=1, shapeName=dyn_fur_e_l_capsule, center=<0 0 1.4163>}]
 dyn_fur_e_r [{shapeType=Capsule, length=2, radius=1, shapeName=dyn_fur_e_r_capsule, center=<0 0 2.3821>}]
 dyn_fur_n [{shapeType=Capsule, length=2, radius=1, shapeName=dyn_fur_n_capsule, center=<0 0 -2.1751>}]
 dyn_fur_n_l [{shapeType=Capsule, length=2, radius=1, shapeName=dyn_fur_n_l_capsule, center=<0 0 -2.8421>}]
 dyn_fur_s [{shapeType=Capsule, length=2, radius=1, shapeName=dyn_fur_s_capsule, center=<0 0 -3.1005>}]
 dyn_fur_s_l [{shapeType=Capsule, length=2, radius=1, shapeName=dyn_fur_s_l_capsule, center=<0 0 1.8679>}]
 dyn_fur_w [{shapeType=Capsule, length=2, radius=1, shapeName=dyn_fur_w_capsule, center=<0 0 -2.4676>}]
 dyn_fur_w_l [{shapeType=Capsule, length=2, radius=1, shapeName=dyn_fur_w_l_capsule, center=<0 0 -2.4689>}]
 lowerarm_twist_01_l [{
   shapeType=Capsule
   length=11.1106
   radius=0.7263
   shapeName=lowerarm_twist_01_l_capsule
   center=<-0.457 0.3603 -0.0078>
   rotation=<rot 14.3322 69.5523 -90>
  }]
 lowerarm_twist_01_r [{
   shapeType=Capsule
   length=11.1306
   radius=0.7394
   shapeName=lowerarm_twist_01_r_capsule
   center=<0.4623 -0.3492 0.0193>
   rotation=<rot 15.4444 68.8554 -90>
  }]
}
constraints [
 {bone1Name=dyn_fur_s_l, bone2Name=lowerarm_twist_01_l, swing1LimitDegrees=30, swing2LimitDegrees=10, twistLimitDegrees=2}
 {bone1Name=dyn_fur_n_l, bone2Name=lowerarm_twist_01_l, swing1LimitDegrees=30, swing2LimitDegrees=10, twistLimitDegrees=2}
 {bone1Name=dyn_fur_w_l, bone2Name=lowerarm_twist_01_l, swing1LimitDegrees=30, swing2LimitDegrees=10, twistLimitDegrees=2}
 {bone1Name=dyn_fur_e_l, bone2Name=lowerarm_twist_01_l, swing1LimitDegrees=30, swing2LimitDegrees=10, twistLimitDegrees=2}
 {bone1Name=dyn_fur_s, bone2Name=lowerarm_twist_01_r, swing1LimitDegrees=30, swing2LimitDegrees=10, twistLimitDegrees=2}
 {bone1Name=dyn_fur_e_r, bone2Name=lowerarm_twist_01_r, swing1LimitDegrees=30, swing2LimitDegrees=10, twistLimitDegrees=2}
 {bone1Name=dyn_fur_n, bone2Name=lowerarm_twist_01_r, swing1LimitDegrees=30, swing2LimitDegrees=10, twistLimitDegrees=2}
 {bone1Name=dyn_fur_w, bone2Name=lowerarm_twist_01_r, swing1LimitDegrees=30, swing2LimitDegrees=10, twistLimitDegrees=2}
]
