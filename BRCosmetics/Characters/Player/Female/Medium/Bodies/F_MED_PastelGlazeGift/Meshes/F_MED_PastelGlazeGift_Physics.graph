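# Physics asset for the PastelGlazeGift female medium cosmetic body, covering its dyn_fur bones around the upper spine.
# Ten fur strands (front, back, four left, four right) each pair a small base capsule with a tip capsule under tight swing and twist limits, narrower swing on the third side strands.
# All bases attach to spine_05 with free rotation, and spine_05 links likewise to a large, zero-length spine_04 capsule.
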PhysicsAsset /BRCosmetics/Characters/Player/Female/Medium/Bodies/F_MED_PastelGlazeGift/Meshes/F_MED_PastelGlazeGift_Physics
bodies {
 dyn_fur_back [{shapeType=Capsule, length=1, radius=1, shapeName=dyn_fur_back_capsule, center=<0.6736 0 0>}]
 dyn_fur_back_base [{shapeType=Capsule, length=0.51, radius=0.5, shapeName=dyn_fur_back_base_capsule, rotation=<rot 0 -90 0>}]
 dyn_fur_front [{shapeType=Capsule, length=1, radius=1, shapeName=dyn_fur_front_capsule, center=<0.6736 0 0>}]
 dyn_fur_front_base [{shapeType=Capsule, length=0.51, radius=0.5, shapeName=dyn_fur_front_base_capsule, rotation=<rot 0 -90 0>}]
 dyn_fur_l_01 [{shapeType=Capsule, length=1, radius=1, shapeName=dyn_fur_l_01_capsule, center=<0.6736 0 0>}]
 dyn_fur_l_01_base [{shapeType=Capsule, length=0.51, radius=0.5, shapeName=dyn_fur_l_01_base_capsule, rotation=<rot 0 -90 0>}]
 dyn_fur_l_02 [{shapeType=Capsule, length=1, radius=1, shapeName=dyn_fur_l_02_capsule, center=<0.6736 0 0>}]
 dyn_fur_l_02_base [{shapeType=Capsule, length=0.51, radius=0.5, shapeName=dyn_fur_l_02_base_capsule, rotation=<rot 0 -90 0>}]
 dyn_fur_l_03 [{shapeType=Capsule, length=1, radius=1, shapeName=dyn_fur_l_03_capsule, center=<0.6736 0 0>}]
 dyn_fur_l_03_base [{shapeType=Capsule, length=0.51, radius=0.5, shapeName=dyn_fur_l_03_base_capsule, rotation=<rot 0 -90 0>}]
 dyn_fur_l_04 [{shapeType=Capsule, length=1, radius=1, shapeName=dyn_fur_l_04_capsule, center=<0.6736 0 0>}]
 dyn_fur_l_04_base [{shapeType=Capsule, length=0.51, radius=0.5, shapeName=dyn_fur_l_04_base_capsule, rotation=<rot 0 -90 0>}]
 dyn_fur_r_01 [{shapeType=Capsule, length=1, radius=1, shapeName=dyn_fur_r_01_capsule, center=<0.6736 0 0>}]
 dyn_fur_r_01_base [{shapeType=Capsule, length=0.51, radius=0.5, shapeName=dyn_fur_r_01_base_capsule, rotation=<rot 0 -90 0>}]
 dyn_fur_r_02 [{shapeType=Capsule, length=1, radius=1, shapeName=dyn_fur_r_02_capsule, center=<0.6736 0 0>}]
 dyn_fur_r_02_base [{shapeType=Capsule, length=0.51, radius=0.5, shapeName=dyn_fur_r_02_base_capsule, rotation=<rot 0 -90 0>}]
 dyn_fur_r_03 [{shapeType=Capsule, length=1, radius=1, shapeName=dyn_fur_r_03_capsule, center=<0.6736 0 0>}]
 dyn_fur_r_03_base [{shapeType=Capsule, length=0.51, radius=0.5, shapeName=dyn_fur_r_03_base_capsule, rotation=<rot 0 -90 0>}]
 dyn_fur_r_04 [{shapeType=Capsule, length=1, radius=1, shapeName=dyn_fur_r_04_capsule, center=<0.6736 0 0>}]
 dyn_fur_r_04_base [{shapeType=Capsule, length=0.51, radius=0.5, shapeName=dyn_fur_r_04_base_capsule, rotation=<rot 0 -90 0>}]
 spine_04 [{shapeType=Capsule, length=0, radius=11.8612, shapeName=spine_04_capsule, center=<-0.9685 2.6044 -0.2082>, rotation=<rot 1.7126 -90 -0.2673>}]
 spine_05 [{shapeType=Capsule, length=0.7127, radius=2, shapeName=spine_05_capsule, center=<3.8835 0.857 -0.3192>, rotation=<rot 0 -164.6326 -82.2012>}]
}
constraints [
 {bone1Name=dyn_fur_back, bone2Name=dyn_fur_back_base, swing1LimitDegrees=20, swing2LimitDegrees=3, twistLimitDegrees=2}
 {bone1Name=dyn_fur_front, bone2Name=dyn_fur_front_base, swing1LimitDegrees=20, swing2LimitDegrees=3, twistLimitDegrees=2}
 {bone1Name=dyn_fur_l_01, bone2Name=dyn_fur_l_01_base, swing1LimitDegrees=20, swing2LimitDegrees=3, twistLimitDegrees=2}
 {bone1Name=dyn_fur_l_02, bone2Name=dyn_fur_l_02_base, swing1LimitDegrees=20, swing2LimitDegrees=3, twistLimitDegrees=2}
 {bone1Name=dyn_fur_l_03, bone2Name=dyn_fur_l_03_base, swing1LimitDegrees=10, swing2LimitDegrees=3, twistLimitDegrees=2}
 {bone1Name=dyn_fur_l_04, bone2Name=dyn_fur_l_04_base, swing1LimitDegrees=20, swing2LimitDegrees=3, twistLimitDegrees=2}
 {bone1Name=dyn_fur_r_01, bone2Name=dyn_fur_r_01_base, swing1LimitDegrees=20, swing2LimitDegrees=3, twistLimitDegrees=2}
 {bone1Name=dyn_fur_r_02, bone2Name=dyn_fur_r_02_base, swing1LimitDegrees=20, swing2LimitDegrees=3, twistLimitDegrees=2}
 {bone1Name=dyn_fur_r_03, bone2Name=dyn_fur_r_03_base, swing1LimitDegrees=10, swing2LimitDegrees=3, twistLimitDegrees=2}
 {bone1Name=dyn_fur_r_04, bone2Name=dyn_fur_r_04_base, swing1LimitDegrees=20, swing2LimitDegrees=3, twistLimitDegrees=2}
 {
  bone1Name=spine_05
  bone2Name=spine_04
  swing1LimitDegrees=45
  swing1Motion=Free
  swing2LimitDegrees=45
  swing2Motion=Free
  twistLimitDegrees=45
  twistMotion=Free
 }
 {
  bone1Name=dyn_fur_back_base
  bone2Name=spine_05
  swing1LimitDegrees=45
  swing1Motion=Free
  swing2LimitDegrees=45
  swing2Motion=Free
  twistLimitDegrees=45
  twistMotion=Free
 }
 {
  bone1Name=dyn_fur_front_base
  bone2Name=spine_05
  swing1LimitDegrees=45
  swing1Motion=Free
  swing2LimitDegrees=45
  swing2Motion=Free
  twistLimitDegrees=45
  twistMotion=Free
 }
 {
  bone1Name=dyn_fur_l_01_base
  bone2Name=spine_05
  swing1LimitDegrees=45
  swing1Motion=Free
  swing2LimitDegrees=45
  swing2Motion=Free
  twistLimitDegrees=45
  twistMotion=Free
 }
 {
  bone1Name=dyn_fur_l_02_base
  bone2Name=spine_05
  swing1LimitDegrees=45
  swing1Motion=Free
  swing2LimitDegrees=45
  swing2Motion=Free
  twistLimitDegrees=45
  twistMotion=Free
 }
 {
  bone1Name=dyn_fur_l_03_base
  bone2Name=spine_05
  swing1LimitDegrees=45
  swing1Motion=Free
  swing2LimitDegrees=45
  swing2Motion=Free
  twistLimitDegrees=45
  twistMotion=Free
 }
 {
  bone1Name=dyn_fur_l_04_base
  bone2Name=spine_05
  swing1LimitDegrees=45
  swing1Motion=Free
  swing2LimitDegrees=45
  swing2Motion=Free
  twistLimitDegrees=45
  twistMotion=Free
 }
 {
  bone1Name=dyn_fur_r_01_base
  bone2Name=spine_05
  swing1LimitDegrees=45
  swing1Motion=Free
  swing2LimitDegrees=45
  swing2Motion=Free
  twistLimitDegrees=45
  twistMotion=Free
 }
 {
  bone1Name=dyn_fur_r_02_base
  bone2Name=spine_05
  swing1LimitDegrees=45
  swing1Motion=Free
  swing2LimitDegrees=45
  swing2Motion=Free
  twistLimitDegrees=45
  twistMotion=Free
 }
 {
  bone1Name=dyn_fur_r_03_base
  bone2Name=spine_05
  swing1LimitDegrees=45
  swing1Motion=Free
  swing2LimitDegrees=45
  swing2Motion=Free
  twistLimitDegrees=45
  twistMotion=Free
 }
 {
  bone1Name=dyn_fur_r_04_base
  bone2Name=spine_05
  swing1LimitDegrees=45
  swing1Motion=Free
  swing2LimitDegrees=45
  swing2Motion=Free
  twistLimitDegrees=45
  twistMotion=Free
 }
]
