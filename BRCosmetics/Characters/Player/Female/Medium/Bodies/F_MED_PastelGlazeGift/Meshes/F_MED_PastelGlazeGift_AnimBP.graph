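# Post-process style AnimBlueprint for the PastelGlazeGift female medium body cosmetic, adding secondary motion to the pose copied from the attached parent mesh.
# Control Rig feeds rotation constraints for fur and belt-buckle bones, then two spine_05-space RigidBody passes, the second using a hand-cuff physics asset.
# Each update sets sim space from DBNO/front-end state, a deprecated gravity override and remapped spine_03 yaw, though no AnimGraph pin visibly reads these variables.
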AnimBlueprint /BRCosmetics/Characters/Player/Female/Medium/Bodies/F_MED_PastelGlazeGift/Meshes/F_MED_PastelGlazeGift_AnimBP
target_skeleton=&/F_MED_PastelGlazeGift_Skeleton
graphs [
 AnimGraph {nodes [
   n1 = Control Rig(Source=Copy Pose From Mesh() {bUseAttachedParent=true}) {
    alphaBoolBlend {blendOption=Linear}
    controlRigAssetReference {blueprintRigClass=&/F_MED_PastelGlazeGift_CtrlRig.F_MED_PastelGlazeGift_CtrlRig_C}
   }
   n2 = "Constraint - dyn_fur_l_03_base (R)"(ComponentPose=Local To Component(LocalPose=@n1), Alpha=1, ConstraintWeights_0=1) {
    alphaBoolBlend {blendOption=Linear}
    boneToModify {boneName=dyn_fur_l_03_base}
    constraintSetup [{transformType=Rotation, targetBone {boneName=clavicle_l}}]
    constraintWeights [1]
   }
   n3 = "Constraint - belt_buckle_correction_base (RR)"!(ComponentPose=@n2, Alpha=1, ConstraintWeights_0=0.8, ConstraintWeights_1:float=0.2) -> (Pose) {
    alphaBoolBlend {blendOption=Linear}
    boneToModify {boneName=belt_buckle_correction_base}
    constraintSetup [{transformType=Rotation, targetBone {boneName=pelvis}}, {transformType=Rotation, targetBone {boneName=spine_01}}]
    constraintWeights [1,1]
   }
   n4 = RigidBody!(ComponentPose=@n3, bAlphaBoolEnabled:bool, OverrideWorldGravity:Vector="0,0,-980", ExternalForce) -> (Pose) {
    alphaInputType=Bool
    bClampLinearTranslationLimitToRefPose=true
    bOverrideWorldGravity=true
    overridePhysicsAsset=&/F_MED_PastelGlazeGift_Physics
    simulationSpace=BaseBoneSpace
    alphaBoolBlend {blendOption=Linear}
    baseBoneRef {boneName=spine_05}
    componentLinearAccScale=<0.6 0.8 0.6>
    componentLinearVelScale=<0.6 0.8 0.6>
    simSpaceSettings {maxAngularAcceleration=30, maxAngularVelocity=3, maxLinearAcceleration=20, maxLinearVelocity=1, velocityScaleZ=0.2, worldAlpha=0.8}
   }
   n5 = RigidBody!(ComponentPose=@n4, Alpha:float=0.8, ExternalForce) -> (Pose) {
    overridePhysicsAsset=&/F_MED_PastelGlazeGift_Hand_Cuffs_Physics
    simulationSpace=BaseBoneSpace
    alphaBoolBlend {blendOption=Linear}
    baseBoneRef {boneName=spine_05}
   }
   Output Pose(Result=Component To Local(ComponentPose=@n5)) {layerGroup=None}
  ]}
 EventGraph {nodes [
   Event BlueprintUpdateAnimation()
   > n1 = Sequence!(execute) -> (then_0, then_1, then_2:Exec)
   > Set simSpace(simSpace:AnimPhysSimSpaceType Enum) -> (Output_Get:AnimPhysSimSpaceType Enum)
   > Branch(Condition=OR Boolean(A=Get bIsDBNO(), B=Get bIsInFrontEnd()))
   > n2 = Branch(Condition=OR Boolean(A=Get bIsDBNOCarried(), B=Get bIsInFrontEnd()))
   > Set simSpace(simSpace=World)
   Set simSpace(execute=@n2.else)
   n3 = Gravity Override State Function DEPRECATED(
    execute=@n1.then_1
    CharacterPartAnimInstance=Self-Reference()
    GameTimeInSeconds=GetGameTimeInSeconds()
    GravityOverrideParams=Get Gravity Override Params()
   )
   > Set Gravity Override(Gravity Override=@n3)
   n4 = GetOwnerSkeletalMeshComponent()
   Set spine_3_rotation(
    execute=@n1.then_2
    spine_3_rotation:double=MapRangeClamped(
     Value=GetDeltaTransformFromRefPose!(self=@n4, BoneName=spine_03, BaseName=pelvis) -> (
      ReturnValue_Location:Vector
      ReturnValue_Rotation_Roll:float
      ReturnValue_Rotation_Pitch:float
      ReturnValue_Rotation_Yaw:float
      ReturnValue_Scale:Vector
     ).ReturnValue_Rotation_Yaw
     InRangeA=10
     InRangeB=35
    )
   ) -> (Output_Get:double)
  ]}
]
variables [
 Gravity Override {type=Vector}
 Gravity Override Params {type=GravityOverrideParamsStruct}
 simSpace {type=AnimPhysSimSpaceType}
 spine_3_rotation {type=number}
]
