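# Physics asset for the back-hair part of the female PastelGlazeGift character, defining collision bodies and joints for hair dynamics.
# A branching chain runs from Base_hair_back_L through L_01 to L_03, forking into L_04/L_07 and L_05/L_08, with an R_01–R_03 branch off L_02.
# Two curl bones hang off L_01 with twist and swing1 locked.
# Unconstrained capsules on pelvis, spine, clavicles and upper arms provide body collision shapes.
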PhysicsAsset /BRCosmetics/Characters/Player/Female/Medium/Bodies/F_MED_PastelGlazeGift/Meshes/Parts/F_MED_PastelGlazeGift_FaceAcc_BackHair_Physics
bodies {
 Base_hair_back_L [{shapeType=Sphere, radius=1.0734, shapeName=Base_hair_back_L_sphere}]
 Dyn_hair_back_L_01 [{
   shapeType=Capsule
   length=24.2618
   radius=4.456
   shapeName=Dyn_hair_back_L_01_capsule
   center=<3.8171 -0.2243 1.6504>
   rotation=<rot -0.0006 0 -89.8832>
  }]
 Dyn_hair_back_L_02 [
  {
   shapeType=Capsule
   length=7.9193
   radius=4.01
   shapeName=Dyn_hair_back_L_02_capsule
   center=<4.299 -0.3155 2.1562>
   rotation=<rot -0.1256 90.7293 -81.2884>
  }
  {
   shapeType=Capsule
   length=7.621
   radius=3.57
   shapeName=Dyn_hair_back_L_02_capsule
   center=<3.9305 8.9592 0.8142>
   rotation=<rot -81.9092 0 0>
  }
  {
   shapeType=Capsule
   length=7.82
   radius=4.048
   shapeName=Dyn_hair_back_L_02_capsule
   center=<4.358 -8.776 1.2397>
   rotation=<rot -86.0036 0 0>
  }
 ]
 Dyn_hair_back_L_03 [{shapeType=Box, extentX=7.5, extentY=24.7497, extentZ=5.7421, shapeName=Dyn_hair_back_L_03_box, center=<3.4595 -1.4286 -0.055>}]
 Dyn_hair_back_L_04 [{
   shapeType=Box
   extentX=12.1692
   extentY=22.1413
   extentZ=7.3812
   shapeName=Dyn_hair_back_L_04_box
   center=<4.123 -0.4547 0.3498>
   rotation=<rot 5.3719 0 0>
  }]
 Dyn_hair_back_L_05 [{
   shapeType=Capsule
   length=7.8281
   radius=3.6883
   shapeName=Dyn_hair_back_L_08_capsule
   center=<5.0274 0.058 0.3823>
   rotation=<rot 3.161 -178.3071 -86.0411>
  }]
 Dyn_hair_back_L_07 [{
   shapeType=Box
   extentX=10.1692
   extentY=20.1413
   extentZ=8.3812
   shapeName=Dyn_hair_back_L_04_box
   center=<3.0518 0 0.287>
   rotation=<rot 5.3719 0 0>
  }]
 Dyn_hair_back_L_08 [{
   shapeType=Capsule
   length=7.8281
   radius=3.6883
   shapeName=Dyn_hair_back_L_08_capsule
   center=<5.0274 0.058 0.3823>
   rotation=<rot 3.161 -178.3071 -86.0411>
  }]
 Dyn_hair_back_R_01 [{
   shapeType=Capsule
   length=7.8281
   radius=3.6883
   shapeName=Dyn_hair_back_L_08_capsule
   center=<5.0274 0.058 0.3823>
   rotation=<rot 3.161 -178.3071 -86.0411>
  }]
 Dyn_hair_back_R_02 [{
   shapeType=Capsule
   length=7.8281
   radius=3.6883
   shapeName=Dyn_hair_back_L_08_capsule
   center=<5.0274 0.058 0.3823>
   rotation=<rot 3.161 -178.3071 -86.0411>
  }]
 Dyn_hair_back_R_03 [{
   shapeType=Capsule
   length=7.8281
   radius=3.6883
   shapeName=Dyn_hair_back_L_08_capsule
   center=<5.0274 0.058 0.3823>
   rotation=<rot 3.161 -178.3071 -86.0411>
  }]
 clavicle_l [{
   shapeType=Capsule
   length=9.326
   radius=5.6718
   shapeName=clavicle_l_capsule
   center=<6.4229 -1.3784 -2.2501>
   rotation=<rot 86.9638 -0.1986 -5.0246>
  }]
 clavicle_r [{
   shapeType=Capsule
   length=9.326
   radius=5.6718
   shapeName=clavicle_l_capsule
   center=<-6.4725 1.6489 1.745>
   rotation=<rot 86.9638 -0.1986 174.9754>
  }]
 dyn_hairCurlCorr_l [{
   shapeType=Capsule
   length=2.4547
   radius=1.6587
   shapeName=dyn_hairCurlCorr_l_capsule
   center=<2.0332 0.65 -0.2096>
   rotation=<rot 81.0273 7.3311 -13.4567>
  }]
 dyn_hairCurlCorr_r [{
   shapeType=Capsule
   length=2.9388
   radius=1.249
   shapeName=dyn_hairCurlCorr_r_capsule
   center=<0.8914 -0.3353 0.5281>
   rotation=<rot 88.1359 0.0025 -77.7477>
  }]
 pelvis [{shapeType=Capsule, length=10.1199, radius=10.7071, shapeName=pelvis_capsule, center=<-3.6105 -4.1246 0>}]
 spine_01 [{shapeType=Capsule, length=5.9341, radius=10.3359, shapeName=spine_01_capsule, center=<0 0.3179 0>}]
 spine_02 [{shapeType=Capsule, length=3.5085, radius=9.9215, shapeName=spine_02_capsule, center=<0 1.395 0>}]
 spine_03 [{shapeType=Capsule, length=5.8535, radius=10.0365, shapeName=spine_03_capsule, center=<-2.0838 2.2176 0>}]
 spine_04 [{shapeType=Capsule, length=6.7699, radius=9.8801, shapeName=spine_04_capsule, center=<0 2.6192 0>}]
 spine_05 [{shapeType=Capsule, length=11.3136, radius=7.7863, shapeName=spine_05_capsule, center=<0.0464 0.3309 0>}]
 upperarm_l [{
   shapeType=Capsule
   length=24.3425
   radius=5.6575
   shapeName=upperarm_l_capsule
   center=<12.6605 -0.1968 -0.0207>
   rotation=<rot -89.9062 179.9462 -179.9461>
  }]
 upperarm_r [{
   shapeType=Capsule
   length=24.3425
   radius=5.6575
   shapeName=upperarm_l_capsule
   center=<-12.6605 0.1968 0.0207>
   rotation=<rot -89.9062 179.9462 0.0539>
  }]
}
constraints [
 {bone1Name=Dyn_hair_back_L_01, bone2Name=Base_hair_back_L, swing1LimitDegrees=5, swing2LimitDegrees=15, twistLimitDegrees=15}
 {bone1Name=Dyn_hair_back_L_04, bone2Name=Dyn_hair_back_L_03, swing1LimitDegrees=10, swing2LimitDegrees=20, twistLimitDegrees=10}
 {bone1Name=Dyn_hair_back_L_02, bone2Name=Dyn_hair_back_L_01, swing1LimitDegrees=1, swing2LimitDegrees=40, twistLimitDegrees=15}
 {
  bone1Name=Dyn_hair_back_L_03
  bone2Name=Dyn_hair_back_L_02
  swing1LimitDegrees=20
  swing2LimitDegrees=35
  twistLimitDegrees=20
  twistMotion=Locked
 }
 {
  bone1Name=dyn_hairCurlCorr_l
  bone2Name=Dyn_hair_back_L_01
  swing1LimitDegrees=45
  swing1Motion=Locked
  swing2LimitDegrees=60
  twistLimitDegrees=45
  twistMotion=Locked
 }
 {
  bone1Name=dyn_hairCurlCorr_r
  bone2Name=Dyn_hair_back_L_01
  swing1LimitDegrees=45
  swing1Motion=Locked
  swing2LimitDegrees=20
  twistLimitDegrees=45
  twistMotion=Locked
 }
 {bone1Name=Dyn_hair_back_L_07, bone2Name=Dyn_hair_back_L_04, swing1LimitDegrees=10, swing2LimitDegrees=15, twistLimitDegrees=10}
 {
  bone1Name=Dyn_hair_back_L_08
  bone2Name=Dyn_hair_back_L_05
  swing1LimitDegrees=10
  swing2LimitDegrees=30
  twistLimitDegrees=45
  twistMotion=Locked
 }
 {
  bone1Name=Dyn_hair_back_L_05
  bone2Name=Dyn_hair_back_L_03
  swing1LimitDegrees=10
  swing2LimitDegrees=30
  twistLimitDegrees=45
  twistMotion=Locked
 }
 {bone1Name=Dyn_hair_back_R_01, bone2Name=Dyn_hair_back_L_02, swing1LimitDegrees=10, swing2LimitDegrees=30, twistLimitDegrees=0}
 {bone1Name=Dyn_hair_back_R_02, bone2Name=Dyn_hair_back_R_01, swing1LimitDegrees=10, swing2LimitDegrees=25, twistLimitDegrees=0}
 {bone1Name=Dyn_hair_back_R_03, bone2Name=Dyn_hair_back_R_02, swing1LimitDegrees=10, swing2LimitDegrees=20, twistLimitDegrees=0}
]
