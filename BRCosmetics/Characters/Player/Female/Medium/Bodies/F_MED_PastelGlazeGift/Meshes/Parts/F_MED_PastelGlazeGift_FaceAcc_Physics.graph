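# Physics asset for the face accessory part of the female medium PastelGlazeGift character cosmetic, covering a left front hair strand.
# A sphere at the hair base anchors a chain of four capsule bodies, linked by constraints with locked twist and swing limit angles that widen toward the tip.
# Unconstrained capsules on the upper spine, clavicles and upper arms supply torso and shoulder shapes for the hair to collide with.
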PhysicsAsset /BRCosmetics/Characters/Player/Female/Medium/Bodies/F_MED_PastelGlazeGift/Meshes/Parts/F_MED_PastelGlazeGift_FaceAcc_Physics
bodies {
 Base_hair_Front_L [{shapeType=Sphere, radius=1.6809, shapeName=Base_hair_Front_L_sphere}]
 Dyn_hair_Front_L_01 [{
   shapeType=Capsule
   length=0.5
   radius=2.8647
   shapeName=Dyn_hair_Front_L_01_capsule
   center=<3.8765 0.003 3.485>
   rotation=<rot -81.9208 85.8604 4.3563>
  }]
 Dyn_hair_Front_L_02 [{
   shapeType=Capsule
   length=4.3501
   radius=3
   shapeName=Dyn_hair_Front_L_02_capsule
   center=<3.7675 0 2.2378>
   rotation=<rot -74.7872 179.9935 -179.991>
  }]
 Dyn_hair_Front_L_03 [{
   shapeType=Capsule
   length=3.1451
   radius=3
   shapeName=Dyn_hair_Front_L_03_capsule
   center=<3.5057 0.0001 0.8145>
   rotation=<rot 88.8689 -179.9943 -179.9956>
  }]
 Dyn_hair_Front_L_04 [{
   shapeType=Capsule
   length=6.7382
   radius=1.9878
   shapeName=Dyn_hair_Front_L_04_capsule
   center=<4.5097 0.1991 -1.2148>
   rotation=<rot 71.532 -52.7319 51.7348>
  }]
 clavicle_l [{
   shapeType=Capsule
   length=4.6779
   radius=6.1717
   shapeName=clavicle_l_capsule
   center=<6.3931 -1.0251 -2.2485>
   rotation=<rot 86.9643 -0.0014 -0.0017>
  }]
 clavicle_r [{
   shapeType=Capsule
   length=4.6779
   radius=6.1717
   shapeName=clavicle_l_capsule
   center=<-6.3931 1.0251 2.2485>
   rotation=<rot 86.9643 -0.0014 179.9983>
  }]
 spine_03 [{shapeType=Capsule, length=4.8535, radius=9.5365, shapeName=spine_03_capsule, center=<-2.0838 2.2176 0>}]
 spine_04 [{shapeType=Capsule, length=4.2699, radius=10.8801, shapeName=spine_04_capsule, center=<0 2.6192 0>}]
 spine_05 [{shapeType=Capsule, length=9.7422, radius=9.3577, shapeName=spine_05_capsule, center=<0.0464 -1.6321 0>}]
 upperarm_l [{
   shapeType=Capsule
   length=24.2198
   radius=5.7802
   shapeName=upperarm_l_capsule
   center=<12.6605 -0.1968 -0.0207>
   rotation=<rot -89.9062 179.9462 -179.9461>
  }]
 upperarm_r [{
   shapeType=Capsule
   length=24.7198
   radius=5.2802
   shapeName=upperarm_l_capsule
   center=<-12.6605 0.1968 0.0207>
   rotation=<rot -89.9062 179.9462 0.0539>
  }]
}
constraints [
 {
  bone1Name=Dyn_hair_Front_L_02
  bone2Name=Dyn_hair_Front_L_01
  swing1LimitDegrees=10
  swing2LimitDegrees=25
  twistLimitDegrees=45
  twistMotion=Locked
 }
 {
  bone1Name=Dyn_hair_Front_L_03
  bone2Name=Dyn_hair_Front_L_02
  swing1LimitDegrees=10
  swing2LimitDegrees=30
  twistLimitDegrees=45
  twistMotion=Locked
 }
 {
  bone1Name=Dyn_hair_Front_L_04
  bone2Name=Dyn_hair_Front_L_03
  swing1LimitDegrees=10
  swing2LimitDegrees=60
  twistLimitDegrees=45
  twistMotion=Locked
 }
 {
  bone1Name=Dyn_hair_Front_L_01
  bone2Name=Base_hair_Front_L
  swing1LimitDegrees=5
  swing2LimitDegrees=10
  twistLimitDegrees=45
  twistMotion=Locked
 }
]
