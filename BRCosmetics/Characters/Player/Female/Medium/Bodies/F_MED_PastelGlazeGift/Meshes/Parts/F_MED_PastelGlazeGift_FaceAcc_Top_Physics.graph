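# Physics asset for the hair on the FaceAcc_Top mesh part of the female medium PastelGlazeGift cosmetic body.
# Capsule bodies cover the hair bones and the head.
# Front hang and left bang attach to the head with tight angular limits, while the top bang forms a three-bone chain whose swing limits widen toward the tip.
# Back corrective bones and the bottom hang have all angular motion locked.
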PhysicsAsset /BRCosmetics/Characters/Player/Female/Medium/Bodies/F_MED_PastelGlazeGift/Meshes/Parts/F_MED_PastelGlazeGift_FaceAcc_Top_Physics
bodies {
 dyn_hairCorrBack [{shapeType=Capsule, length=0.505, radius=0.505, shapeName=dyn_hairCorrBack_capsule, rotation=<rot 0 -90 0>}]
 dyn_hair_bang_left [{
   shapeType=Capsule
   length=0.7666
   radius=2.1341
   shapeName=dyn_hair_bang_left_capsule
   center=<2.0515 0.763 0.464>
   rotation=<rot 0 -152.139 -85.9885>
  }]
 dyn_hair_bang_top_01 [{
   shapeType=Capsule
   length=0.505
   radius=0.505
   shapeName=dyn_hair_bang_top_01_capsule
   center=<1.6709 -0.0279 0.0944>
   rotation=<rot -76.6354 179.9998 -179.9997>
  }]
 dyn_hair_bang_top_02 [{
   shapeType=Capsule
   length=2.4856
   radius=1
   shapeName=dyn_hair_bang_top_02_capsule
   center=<1.7061 0.0955 0.0017>
   rotation=<rot 0.9162 -93.9468 -87.5475>
  }]
 dyn_hair_hang_bot [{shapeType=Capsule, length=3.5481, radius=2.4089, shapeName=dyn_hair_hang_bot_capsule, rotation=<rot 0 86.2638 -80.6009>}]
 dyn_hair_hang_front [{shapeType=Capsule, length=7.005, radius=0.505, shapeName=dyn_hair_hang_front_capsule, center=<4.339 0.446 0>, rotation=<rot 0 -90 109.8246>}]
 dyn_hair_hang_top_03 [{
   shapeType=Capsule
   length=2.2005
   radius=0.8774
   shapeName=dyn_hair_hang_top_03_capsule
   center=<2.2956 -0.8291 -0.049>
   rotation=<rot 0 -146.2793 -87.6488>
  }]
 hairCorrBack [{shapeType=Capsule, length=0.505, radius=0.505, shapeName=hairCorrBack_capsule, rotation=<rot 0 -90 0>}]
 head [{shapeType=Capsule, length=0, radius=7.0738, center=<1.1913 -2.6286 -0.6493>, rotation=<rot -33.0816 -90 -22.0185>}]
}
constraints [
 {bone1Name=dyn_hair_hang_front, bone2Name=head, swing1LimitDegrees=10, swing2LimitDegrees=10, twistLimitDegrees=2}
 {bone1Name=dyn_hair_bang_left, bone2Name=head, swing1LimitDegrees=5, swing2LimitDegrees=15, twistLimitDegrees=2}
 {bone1Name=dyn_hair_bang_top_01, bone2Name=head, swing1LimitDegrees=1, swing2LimitDegrees=5, twistLimitDegrees=2}
 {bone1Name=dyn_hair_bang_top_02, bone2Name=dyn_hair_bang_top_01, swing1LimitDegrees=3, swing2LimitDegrees=10, twistLimitDegrees=2}
 {bone1Name=dyn_hair_hang_top_03, bone2Name=dyn_hair_bang_top_02, swing1LimitDegrees=5, swing2LimitDegrees=20, twistLimitDegrees=2}
 {
  bone1Name=hairCorrBack
  bone2Name=head
  swing1LimitDegrees=45
  swing1Motion=Locked
  swing2LimitDegrees=45
  swing2Motion=Locked
  twistLimitDegrees=45
  twistMotion=Locked
 }
 {
  bone1Name=dyn_hairCorrBack
  bone2Name=hairCorrBack
  swing1LimitDegrees=45
  swing1Motion=Locked
  swing2LimitDegrees=45
  swing2Motion=Locked
  twistLimitDegrees=45
  twistMotion=Locked
 }
 {
  bone1Name=dyn_hair_hang_bot
  bone2Name=head
  swing1LimitDegrees=45
  swing1Motion=Locked
  swing2LimitDegrees=45
  swing2Motion=Locked
  twistLimitDegrees=45
  twistMotion=Locked
 }
]
