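# Post-process-style AnimBP for the PastelGlazeGift female cosmetic's FaceAcc part, animating its dynamic hair.
# Copied body pose feeds RigidBody, Control Rig and AnimDynamics stages, cached as 'pre sim'.
# That cache feeds two back-hair rig and physics branches, selected by backbling or disabled rigid bodies, then a final Control Rig.
# The event graph derives velocity-gated spine twist, frontend/emote flags and backbling-dependent gravity overrides.
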AnimBlueprint /BRCosmetics/Characters/Player/Female/Medium/Bodies/F_MED_PastelGlazeGift/Meshes/Parts/F_MED_PastelGlazeGift_FaceAcc_AnimBP
target_skeleton=&/F_MED_PastelGlazeGift_FaceAcc_Skeleton
graphs [
 AnimGraph {nodes [
   n1 = Get TwistAlphaSpine05() -> (TwistAlphaSpine05:float)
   n2 = Get FrontendAndEmote() -> (FrontendAndEmote:bool)
   n3 = Control Rig!(
    Source=Use cached pose 'pre sim'()
    bAlphaBoolEnabled:bool=True
    AngleSprinting:double=@n1
    isFrontEndAndEmote:bool=@n2
    is skydive:bool=Get bIsSkyDiving()
   ) -> (Pose) {
    alphaInputType=Bool
    alphaBoolBlend {blendOption=Linear}
    controlRigAssetReference {blueprintRigClass=&/F_MED_PastelGlazeGift_FaceAcc_CtrlRig.F_MED_PastelGlazeGift_FaceAcc_CtrlRig_C}
   }
   n4 = RigidBody!(ComponentPose=Local To Component(LocalPose=@n3), bAlphaBoolEnabled:bool, OverrideWorldGravity:Vector="0,0,0", ExternalForce) -> (Pose) {
    alphaInputType=Bool
    bClampLinearTranslationLimitToRefPose=true
    bOverrideWorldGravity=true
    overridePhysicsAsset=&/F_MED_PastelGlazeGift_FaceAcc_BackHair_Physics
    simulationSpace=BaseBoneSpace
    alphaBoolBlend {blendOption=Linear}
    baseBoneRef {boneName=spine_05}
    componentAppliedLinearAccClamp=<6000 8000 8000>
    componentLinearAccScale=<0.35 0.5 0.5>
    componentLinearVelScale=<0.3 0.3 0.3>
    simSpaceSettings {maxAngularAcceleration=30, maxAngularVelocity=3, maxLinearAcceleration=100, maxLinearVelocity=1, velocityScaleZ=0.2, worldAlpha=0.8}
   }
   n5 = Anim Dynamics - Bone: Dyn_hair_back_L_06!(ComponentPose=@n4, bAlphaBoolEnabled:bool=True, GravityOverride:Vector="0,0,0", ExternalForce="0,0,0") -> (Pose) {
    alphaInputType=Bool
    angularDampingOverride=0.75
    angularSpringConstant=250
    bAngularSpring=true
    bGravityOverrideInSimSpace=true
    bOverrideAngularDamping=true
    bUseGravityOverride=true
    simulationSpace=BoneRelative
    alphaBoolBlend {blendOption=Linear}
    boundBone {boneName=Dyn_hair_back_L_06}
    componentLinearAccScale=<2 2 2>
    componentLinearVelScale=<2 2 2>
    physicsBodyDefinitions [{
      boundBone {boneName=Dyn_hair_back_L_06}
      boxExtents=<10 10 10>
      constraintSetup {angularTargetAxis=AxisY, angularLimitsMax=<3 5 0>, angularLimitsMin=<-3 -5 -15>, angularTarget=<0 -1 0>}
      localJointOffset=<0.1 0 0>
     }]
    relativeSpaceBone {boneName=Dyn_hair_back_L_02}
   }
   n6 = Control Rig!(Source=Use cached pose 'pre sim'(), bAlphaBoolEnabled=True, AngleSprinting=@n1, isInFrontendOrEmote:bool=@n2) -> (Pose) {
    alphaInputType=Bool
    alphaBoolBlend {blendOption=Linear}
    controlRigAssetReference {blueprintRigClass=&/F_MED_PastelGlazeGift_FaceAccBackPack_CtrlRig.F_MED_PastelGlazeGift_FaceAccBackPack_CtrlRig_C}
   }
   n7 = RigidBody(ComponentPose=Local To Component(LocalPose=@n6), OverrideWorldGravity="0,0,0") {
    alphaInputType=Bool
    bOverrideWorldGravity=true
    overridePhysicsAsset=&/F_MED_PastelGlazeGift_FaceAcc_BackHairBack_Physics
    simulationSpace=BaseBoneSpace
    alphaBoolBlend {blendOption=Linear}
    baseBoneRef {boneName=spine_05}
    componentLinearAccScale=<0.6 0.6 0.6>
    componentLinearVelScale=<0.6 0.6 0.6>
    simSpaceSettings {maxAngularAcceleration=30, maxAngularVelocity=3, maxLinearAcceleration=10, maxLinearVelocity=100, velocityScaleZ=0.1, worldAlpha=0.8}
   }
   n8 = Control Rig!(
    Source=Blend Poses by bool(
     BlendPose_0=Component To Local(ComponentPose=@n7)
     BlendPose_1=Component To Local(ComponentPose=@n5)
     bActiveValue=Get Is wearing backbling or rigidbody off() -> (Is wearing backbling or rigidbody off:bool)
    ) {blendType=Linear}
    Alpha:float
   ) -> (Pose) {
    alphaBoolBlend {blendOption=Linear}
    controlRigAssetReference {blueprintRigClass=&/F_MED_PastelGlazeGift_FaceAcc_AS_CtrlRig.F_MED_PastelGlazeGift_FaceAcc_AS_CtrlRig_C}
   }
   Output Pose(Result=@n8) {layerGroup=None}
   n9 = Copy Pose From Mesh(SourceMeshComponent=Get Custom Base Mesh() -> (Custom Base Mesh:Skeletal Mesh Component Object Reference)) {bUseAttachedParent=true}
   n10 = Copy Pose From Mesh() {bUseAttachedParent=true}
   n11 = RigidBody(
    ComponentPose=Local To Component(LocalPose=Layered blend per bone(BasePose=@n10, BlendPoses_0=@n9) {layerSetup [{branchFilters [{blendDepth=2, boneName=clavicle_l}, {blendDepth=2, boneName=clavicle_r}]}]})
    OverrideWorldGravity="0,0,0"
   ) {
    alphaInputType=Bool
    bOverrideWorldGravity=true
    overridePhysicsAsset=&/F_MED_PastelGlazeGift_FaceAcc_Top_Physics
    simulationSpace=BaseBoneSpace
    alphaBoolBlend {blendOption=Linear}
    baseBoneRef {boneName=head}
    componentAppliedLinearAccClamp=<7000 7000 7000>
    componentLinearAccScale=<0.7 0.7 0.7>
    componentLinearVelScale=<0.7 0.7 0.7>
    simSpaceSettings {maxAngularAcceleration=30, maxAngularVelocity=3, maxLinearAcceleration=10, maxLinearVelocity=100, velocityScaleZ=0.2, worldAlpha=0.8}
   }
   n12 = Control Rig!(Source=Component To Local(ComponentPose=@n11), bAlphaBoolEnabled:bool=True) -> (Pose) {
    alphaInputType=Bool
    alphaBoolBlend {blendOption=Linear}
    controlRigAssetReference {blueprintRigClass=&/F_MED_PastelGlazeGift_FaceAcc_Frontend_CtrlRig.F_MED_PastelGlazeGift_FaceAcc_Frontend_CtrlRig_C}
   }
   n13 = RigidBody(ComponentPose=Local To Component(LocalPose=@n12), OverrideWorldGravity="0,0,0") {
    alphaInputType=Bool
    bOverrideWorldGravity=true
    overridePhysicsAsset=&/F_MED_PastelGlazeGift_FaceAcc_Physics
    simulationSpace=BaseBoneSpace
    alphaBoolBlend {blendOption=Linear}
    baseBoneRef {boneName=spine_05}
    componentLinearAccScale=<0.6 0.6 0.6>
    componentLinearVelScale=<0.3 0.3 0.3>
    simSpaceSettings {maxAngularAcceleration=30, maxAngularVelocity=3, maxLinearAcceleration=10, maxLinearVelocity=100, velocityScaleZ=0.2, worldAlpha=0.8}
   }
   n14 = Anim Dynamics - Bone: Dyn_hair_Front_L_04!(ComponentPose=@n13, Alpha=1, GravityOverride:Vector="0,0,0", ExternalForce="0,0,0") -> (Pose) {
    angularDampingOverride=0.75
    angularSpringConstant=150
    bAngularSpring=true
    bOverrideAngularDamping=true
    bUseGravityOverride=true
    simulationSpace=BoneRelative
    alphaBoolBlend {blendOption=Linear}
    boundBone {boneName=Dyn_hair_Front_L_04}
    componentAppliedLinearAccClamp=<10000 10000 10000>
    componentLinearAccScale=<2 2 2>
    componentLinearVelScale=<2 2 2>
    physicsBodyDefinitions [{
      boundBone {boneName=Dyn_hair_Front_L_04}
      boxExtents=<10 10 10>
      constraintSetup {angularLimitsMax=<0 0 15>, angularLimitsMin=<0 0 -15>, angularTarget=<1 0 0>}
      localJointOffset=<0.1 0 0>
     }]
    relativeSpaceBone {boneName=Dyn_hair_Front_L_02}
   }
   Save cached pose 'pre sim'(Pose=Component To Local(ComponentPose=@n14)) {cacheName=pre sim}
  ]}
 EventGraph {nodes [
   n1 = Event BlueprintUpdateAnimation()
   > n2 = Is Valid(InputObject=Get PawnOwner() -> (PawnOwner:Pawn Object Reference))
   n3 = BreakVector(InVec=GetVelocity!(self:Actor Object Reference=Get PawnOwner()) -> (ReturnValue:Vector))
   n4 = float + float(
    A=MapRangeClamped(Value="Absolute (Float)"(A=@n3), InRangeA=545, InRangeB=750)
    B=MapRangeClamped(Value="Absolute (Float)"(A=@n3.Y), InRangeA=545, InRangeB=750)
   )
   > Set VelocityXMovement(VelocityXMovement:double=@n4) -> (Output_Get:double)
   n5 = GetAlphaFromReferenceTransformDelta(
    SkeletalMeshComponent=Get Custom Base Mesh()
    BoneName=spine_05
    BaseName=spine_01
    DeltaComponentType=EulerX
    InputRange=(X=-30,Y=30)
    OutputRange=(X=20,Y=-20)
    TwistAxis="1,0,0"
   )
   > n6 = Set TwistAlphaSpine05(TwistAlphaSpine05:float=FInterpTo(
     Current=Get TwistAlphaSpine05()
     Target=float * float(A=@n4, B=@n5) {is=title K2Node_PromotableOperator}
     DeltaTime=@n1.DeltaTimeX
     InterpSpeed=10
    )) -> (Output_Get:float)
   n7 = Gravity Override State Function DEPRECATED(
    execute=[@n6.then, @n2.Is Not Valid]
    CharacterPartAnimInstance=Self-Reference()
    GameTimeInSeconds=GetGameTimeInSeconds()
    GravityOverrideParams=Select(
     Option 0:Gravity Override Params Struct Structure=Get Gravity Override Params_WithoutPack() -> (Gravity Override Params_WithoutPack:Gravity Override Params Struct Structure)
     Option 1:Gravity Override Params Struct Structure=Get Gravity Override Params_WithPack() -> (Gravity Override Params_WithPack:Gravity Override Params Struct Structure)
     Index=Get Is Wearing Backbling()
    ) -> (ReturnValue:Gravity Override Params Struct Structure)
   )
   > Set Gravity Override(Gravity Override=@n7)
   n8 = Get bIsInFrontEnd()
   n9 = Get bIsPlayingEmote()
   > Set NotFrontEndOrEmote(NotFrontEndOrEmote:bool=OR Boolean(A=NOT Boolean(A=@n8), B=@n9)) -> (Output_Get:bool)
   > Set Frontend(Frontend:bool=AND Boolean(A=NOT Boolean(A=@n9), B=@n8)) -> (Output_Get:bool)
   > Set FrontendAndEmote(FrontendAndEmote:bool=AND Boolean(A=@n8, B=@n9)) -> (Output_Get:bool)
   n10 = Get bRigidBodyEnabled()
   > Set Is wearing backbling or rigidbody off(Is wearing backbling or rigidbody off:bool=OR Boolean(A=Get Is Wearing Backbling(), B=NOT Boolean(A=@n10))) -> (Output_Get:bool)
   n11 = Event OnCharacterCustomizationLoaded()
   Set Custom Base Mesh(execute=[@n11.then, @n18.then], Custom Base Mesh:Skeletal Mesh Component Object Reference=@n11.CustomBaseMesh) -> (Output_Get:Skeletal Mesh Component Object Reference)
   n12 = TryGetPawnOwner()
   > n13 = Is Valid(InputObject=@n12)
   n14 = Set PawnOwner(execute=@n13.Is Not Valid, PawnOwner:Pawn Object Reference=TryGetFortPawn()) -> (Output_Get:Pawn Object Reference)
   Cast To FortPlayerPawn(execute=[@n14.then, @n17.then], Object=TryGetFortPawn())
   n15 = Get bRigidBodyEnabled()
   n16 = CheckForBackbling()
   > Set Is Wearing Backbling(Is Wearing Backbling=OR Boolean(A=@n16, B=NOT Boolean(A=@n15)))
   > Set IsNot Wearing Backbling(IsNot Wearing Backbling:bool=AND Boolean(A=NOT Boolean(A=@n16), B=Get bRigidBodyEnabled())) -> (Output_Get:bool)
   n17 = Set PawnOwner(execute=@n13.Is Valid, PawnOwner=@n12)
   n18 = Event BlueprintInitializeAnimation()
  ]}
]
variables [
 Custom Base Mesh {type=SkeletalMeshComponent}
 Frontend {type=boolean}
 FrontendAndEmote {type=boolean}
 Gravity Override {type=Vector}
 Gravity Override Params_WithoutPack {
  type=GravityOverrideParamsStruct
  default {skydive_FloatingDown {gravityOverride=<0 -1000 600>}, skydive_GoingUp {gravityOverride=<0 -1000 600>}}
 }
 Gravity Override Params_WithPack {
  type=GravityOverrideParamsStruct
  default {skydive_FloatingDown {gravityOverride=<0 0 -980>, windAmplitude=<200 -500 0>, windOffset=<0 -300 0>}}
 }
 Is Wearing Backbling {type=boolean}
 Is wearing backbling or rigidbody off {type=boolean}
 IsNot Wearing Backbling {type=boolean}
 NotFrontEndOrEmote {type=boolean}
 PawnOwner {type=Pawn, default=null}
 Twist Alpha Spine 05 {type=number}
 TwistAlphaSpine05 {type=number}
 VelocityXMovement {type=number}
]
